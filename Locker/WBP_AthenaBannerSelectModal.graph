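# Locker banner select modal, a FortBannerSelectModal subclass that hosts the local player banner editor in a scale box inside a UIKit dialog.
# Activation and deactivation pass through to the editor, and activation plays a short fade and scale-down Intro.
# Both close events call HandleCloseAction, which, if closing is allowed, leaves an open tile picker or otherwise pops the modal.
# On mobile, Construct makes the banner fill the overlay with vertical padding.
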WidgetBlueprint /Locker/WBP_AthenaBannerSelectModal
animations [Intro {
  binding_count=1
  length_seconds=0.25
  object=&self:Intro
  sequence {bindings [BP_LocalPlayerBannerEditor {
     id=#1
     possessed_class=WBP_LocalPlayerBannerEditor_C
     tracks [
      {
       type=MovieSceneFloatTrack
       display_name=Render Opacity
       sections [{type=MovieSceneFloatSection, end_frame=5, start_frame=0, channels [None {type=MovieSceneScriptingFloatChannel, keys [[0,0],[2,0],[5,1]]}]}]
      }
      {
       type=MovieScene2DTransformTrack
       sections [{
         type=MovieScene2DTransformSection
         end_frame=5
         start_frame=0
         channels [
          Scale.X {type=MovieSceneScriptingFloatChannel, keys [[2,1.15],[4,1]]}
          Scale.Y {type=MovieSceneScriptingFloatChannel, keys [[2,1.15],[4,1]]}
         ]
        }]
      }
     ]
    }]}
 }]
graphs [
 EventGraph {nodes [
   "Close Banner Editor (BP_LocalPlayerBannerEditor)"!() -> (OutputDelegate, then, IconId:Name=None, ColorId:Name=None) {is=title K2Node_ComponentBoundEvent}
   > HandleCloseAction!(execute, self) -> (then, PassThrough:bool)
   Close() {is=title K2Node_CustomEvent}
   > HandleCloseAction()
   Event On Deactivated()
   > OnWidgetDeactivated(self:WBP Local Player Banner Editor Object Reference=Get BP_LocalPlayerBannerEditor() -> (BP_LocalPlayerBannerEditor:WBP Local Player Banner Editor Object Reference))
   Event On Activated()
   > OnWidgetActivated(self:WBP Local Player Banner Editor Object Reference=Get BP_LocalPlayerBannerEditor())
   > PlayAnimationForward(InAnimation=Get Intro())
   Event Construct()
   > Branch(Condition=IsMobileGame())
   n1 = SlotAsOverlaySlot(Widget=Get ScaleBox_Banner() -> (ScaleBox_Banner:Scale Box Object Reference))
   > SetVerticalAlignment(self=@n1, InVerticalAlignment=VAlign_Fill)
   > SetHorizontalAlignment(self=@n1, InHorizontalAlignment=HAlign_Fill)
   > SetPadding!(execute, self:Overlay Slot Object Reference=@n1, InPadding:Margin Structure=Make Margin(Left=0, Top=25, Right=0, Bottom=25)) -> (then)
  ]}
 HandleCloseAction {nodes [
   HandleCloseAction() {is=title K2Node_FunctionEntry}
   n1 = Get BP_LocalPlayerBannerEditor()
   > n2 = Branch(Condition=Get bCanCloseBannerEditor!(self:Fort Player Profile Banner Editor Object Reference=@n1) -> (bCanCloseBannerEditor:bool))
   > n3 = Branch(Condition=Get bPickingTile!(self:WBP Local Player Banner Editor Object Reference=@n1) -> (bPickingTile:bool))
   > n4 = Attempt Leave Picker!(execute, self:WBP Local Player Banner Editor Object Reference=Get BP_LocalPlayerBannerEditor()) -> (then, Left:bool)
   Return Node!(execute=[@n4.then, @n5.then], PassThrough:bool=false)
   n5 = PopActivatablePanelInModalLayer!(execute=@n3.else, Panel:Common Activatable Panel Legacy Object Reference=Self-Reference()) -> (then)
   Return Node(execute=@n2.else, PassThrough=false)
  ]}
]
variables []
widgets {
 info {
  inheritedWidgetCount=0
  namedSlotCount=1
  parentClass=/Script/FortniteUI.FortBannerSelectModal
  rootWidgetClass=/FortUILibrary/Systems/Dialog/WBP_UIKit_Dialog_Internal.WBP_UIKit_Dialog_Internal_C
  widgetCount=4
 }
 widgets [
  Dialog_Internal {widgetClassPath=/FortUILibrary/Systems/Dialog/WBP_UIKit_Dialog_Internal, bIsVariable=true, widget=Dialog_Internal}
  Overlay_2 {widgetClassPath=&Overlay, namedSlotHost=Dialog_Internal, widget=Overlay_2}
  ScaleBox_Banner {
   widgetClassPath=&ScaleBox
   bIsVariable=true
   parent=Overlay_2
   widget=ScaleBox_Banner
   slot {class=OverlaySlot, properties {horizontalAlignment=HAlign_Center, verticalAlignment=VAlign_Center}}
  }
  BP_LocalPlayerBannerEditor {
   widgetClassPath=/Game/UI/Foundation/Profile/WBP_LocalPlayerBannerEditor
   bIsVariable=true
   parent=ScaleBox_Banner
   widget=BP_LocalPlayerBannerEditor
   slot {class=ScaleBoxSlot}
  }
 ]
}
